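# Sound cue for the firework launcher explosion, layering a medium-range blast with a distant one after a very short random delay.
# Each layer randomly picks one of three waves, then passes through its own Athena rocket-explosion attenuation and slight pitch modulation.
# A separate sound class per layer (third-person fire, distant fire) feeds the mixer.
# The cue joins local and global weapon gunfire concurrency groups to limit stacking.
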SoundCue /Game/Sounds/Fort_Weapons/Launchers/Fireworks/Firework_Launcher_Explosion_Cue
root=Delay_1
nodes [
 {
  id=Attenuation_5
  children [Random_1]
  properties {
   attenuationSettings=/Game/Sounds/Attenuation/Fort_Weapon_Rocket_Distant_Explosion_Athena_Att
   attenuationOverrides {
    distanceAlgorithm=Custom
    falloffDistance=12000
    attenuationShapeExtents {x=0}
    customAttenuationCurve {editorCurveData {keys [
       <at 0 0 RCIM_Cubic>
       {interpMode=RCIM_Cubic, tangentMode=RCTM_User, time=0.12, value=1}
       <at 0.35 1>
       {leaveTangent=-0.2026, time=1, value=0}
      ]}}
   }
  }
 }
 {
  id=Attenuation_6
  children [Random_0]
  properties {attenuationSettings=/Game/Sounds/Attenuation/Fort_Weapon_Rocket_Medium_Explosion_Athena_Att, attenuationOverrides {falloffDistance=1600}}
 }
 {id=Delay_1, children [Mixer_4], properties {delayMax=0.05, delayMin=0.02}}
 {id=Mixer_4, children [SoundClass_6,SoundClass_2], properties {inputVolume [1,1]}}
 {id=Modulator_5, children [Attenuation_5], properties {pitchMax=1.1, pitchMin=0.9}}
 {id=Modulator_6, children [Attenuation_6], properties {pitchMax=1.1, pitchMin=0.9}}
 {id=Random_0, children [WavePlayer_0,WavePlayer_3,WavePlayer_5], properties {weights [1,1,1]}}
 {id=Random_1, children [WavePlayer_6,WavePlayer_7,WavePlayer_8], properties {weights [1,1,1]}}
 {
  id=SoundClass_2
  children [Modulator_5]
  properties {soundClassOverride=&Fort_WeaponRanged_Fire_Distant}
 }
 {id=SoundClass_6, children [Modulator_6], properties {soundClassOverride=&Fort_WeaponRanged_Fire_3P_PSM}}
 {id=WavePlayer_0, properties {soundWaveAssetPtr=&/Firework_Launcher_Explosion_01}}
 {id=WavePlayer_3, properties {soundWaveAssetPtr=&/Firework_Launcher_Explosion_02}}
 {id=WavePlayer_5, properties {soundWaveAssetPtr=&/Firework_Launcher_Explosion_03}}
 {id=WavePlayer_6, properties {soundWaveAssetPtr=&/Firework_Launcher_Explosion_Distant_01}}
 {id=WavePlayer_7, properties {soundWaveAssetPtr=&/Firework_Launcher_Explosion_Distant_02}}
 {id=WavePlayer_8, properties {soundWaveAssetPtr=&/Firework_Launcher_Explosion_Distant_03}}
]
properties {
 maxDistance=15000
 volumeMultiplier=0.7
 concurrencyOverrides {maxCount=4, resolutionRule=StopOldest, volumeScale=0.5}
 concurrencySet [&Weapon_Gunfire_Local_Conc, &Weapon_Gunfire_Global_Conc]
}
